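# Unlit alpha-composite volumetric material for the Lifting Aura gadget prop, raymarching an SDF volume (heart by default) with animated noise distortion.
# Local-space box intersection, optionally clipped by scene depth, sets ray start and length, with blue-noise jitter and platform-switched step counts.
# Custom HLSL samples flipbook pseudo-volumes for density and baked shadow/AO, accumulating time-of-day light and transmittance.
# A dynamic-parameter flash brightens luminance-masked color; sphere inputs feed only commented-out code.
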Material /Game/Animation/Props/Gadgets_LiftingAura/Materials/M_VolRayMarchSDFNoise
graph {
 properties {
  bDisableDepthTest=true
  bUsedWithMeshDeformer=true
  bUsedWithNanite=true
  bUsedWithNiagaraMeshParticles=true
  bUsedWithSkeletalMesh=true
  bUsedWithStaticLighting=true
  blendMode=BLEND_AlphaComposite
  pixelDepthOffsetMode=PDOM_Legacy
  shadingModel=MSM_Unlit
  usedShadingModels=Unlit
 }
 nodes [
  n1 = Constant3Vector() {constant=<rgba 0 0 1 0>}
  Add(A=Transform(None=@n1) {transformSourceType=TRANSFORMSOURCE_View}, B=ActorPositionWS())
  n2 = Max(A=Subtract(A=NamedRerouteUsage(), B=NamedRerouteUsage())) {constB=0}
  n3 = Subtract(A=@n2, B=NamedRerouteUsage())
  n4 = NamedRerouteUsage()
  n5 = Divide(
   A=MaterialFunctionCall!("Xb1,PS4"=@n3, "PC-High&&Epic,XSX,PS5"=@n3, "PC-Perf,Android"=@n2, Switch=@n3, "PC-Low(NotPerf)&&Medium"=@n2) {materialFunction=/Game/Effects/Niagara/MasterMaterials/Functions/Common/PlatformSwitch_3D}.Result
   B=@n4
  )
  Ceil(None=@n5)
  MaterialExpressionNamedRerouteUsage_21 = NamedRerouteUsage()
  MaterialExpressionNamedRerouteUsage_20 = NamedRerouteUsage()
  MaterialExpressionNamedRerouteUsage_3 = NamedRerouteUsage()
  MaterialExpressionNamedRerouteUsage_23 = NamedRerouteUsage()
  MaterialExpressionNamedRerouteUsage_24 = NamedRerouteUsage()
  MaterialExpressionTextureObjectParameter_1 = TextureObjectParameter() {parameterName=Shadow, samplerType=SAMPLERTYPE_LinearColor, texture=&/M_VolumeEncode_Baked}
  MaterialExpressionScalarParameter_1 = ScalarParameter() {defaultValue=30, parameterName=densityGain}
  MaterialExpressionScalarParameter_3 = ScalarParameter() {defaultValue=0.4, parameterName=Shadow Power}
  MaterialExpressionMaterialFunctionCall_4 = MaterialFunctionCall(
   Color_Morning=VectorParameter() {parameterName=Ambient ColorM, defaultValue=<rgba 0.3298 0.123 0.4375 1>}.RGB
   Color_Day=VectorParameter() {parameterName=Ambient Color, defaultValue=<rgba 0.8385 0.1267 0.6782 1>}.RGB
   Color_Evening=VectorParameter() {parameterName=Ambient Color2, defaultValue=<rgba 0.2797 0.0947 0.1868 1>}.RGB
   Color_Night=VectorParameter() {parameterName=Ambient Color3, defaultValue=<rgba 0.0874 0.084 0.2708 1>}.RGB
  ) {materialFunction=&Color_Time_of_Day}
  MaterialExpressionMaterialFunctionCall_3 = MaterialFunctionCall(
   Color_Morning=VectorParameter() {parameterName=Light ColorM, defaultValue=<rgba 0.5873 0.112 0.5938 1>}.RGB
   Color_Day=VectorParameter() {parameterName=Light Color, defaultValue=<rgba 0.9062 0.1369 0.4969 1>}.RGB
   Color_Evening=VectorParameter() {parameterName=Light Color 2, defaultValue=<rgba 0.75 0.0952 0.2846 1>}.RGB
   Color_Night=VectorParameter() {parameterName=Light Color3, defaultValue=<rgba 0.3606 0.0848 0.4746 1>}.RGB
  ) {materialFunction=&Color_Time_of_Day}
  MaterialExpressionSkyAtmosphereLightDirection_1 = SkyAtmosphereLightDirection()
  MaterialExpressionTextureObjectParameter_0 = TextureObjectParameter() {parameterName=Noise, texture=/Game/Animation/Props/Gadgets_LiftingAura/Textures/T_Volume_PerlinWorley}
  MaterialExpressionTime_0 = Time()
  MaterialExpressionMultiply_6 = Multiply(A=@MaterialExpressionTime_0, B=ScalarParameter() {defaultValue=0.05, parameterName=NoiseTime}) {constB=0.015}
  MaterialExpressionScalarParameter_9 = ScalarParameter() {defaultValue=0.125, parameterName=NoiseStrength}
  MaterialExpressionScalarParameter_7 = ScalarParameter() {defaultValue=0.31, parameterName=SphereRadius}
  MaterialExpressionVectorParameter_2 = VectorParameter() {parameterName=SpherePosition, defaultValue=<rgba 0.5 0.2 0.475 1>}
  MaterialExpressionScalarParameter_8 = ScalarParameter() {
   defaultValue=0.666
   desc=0-1 Value. 1 is linear falloff from edge to sphere origin. 0 is hard edge sphere
   parameterName=NormalizedSphereFalloff
  }
  MaterialExpressionTextureObjectParameter_3 = TextureObjectParameter() {parameterName=Stream, samplerType=SAMPLERTYPE_LinearColor, texture=/Game/Animation/Props/Gadgets_LiftingAura/Textures/VT_Heart_SDF_V01}
  MaterialExpressionScalarParameter_10 = ScalarParameter() {defaultValue=1, parameterName=AO Power}
  MaterialExpressionScalarParameter_11 = ScalarParameter() {defaultValue=0.4, parameterName=Inner Softness Distance}
  MaterialExpressionScalarParameter_12 = ScalarParameter() {defaultValue=0.8, parameterName=Density Darken Min}
  MaterialExpressionScalarParameter_13 = ScalarParameter() {defaultValue=2, parameterName=Density Darken Power}
  n6 = Custom!(
   rayStart=@MaterialExpressionNamedRerouteUsage_21
   numSteps=@MaterialExpressionNamedRerouteUsage_20
   cameraVector=@MaterialExpressionNamedRerouteUsage_3
   stepSize=@MaterialExpressionNamedRerouteUsage_23
   finalDelta=@MaterialExpressionNamedRerouteUsage_24
   ShadowTexture=@MaterialExpressionTextureObjectParameter_1
   densityGain=@MaterialExpressionScalarParameter_1
   shadowGain=@MaterialExpressionScalarParameter_3
   ambientColor=@MaterialExpressionMaterialFunctionCall_4.Result
   lightColor=@MaterialExpressionMaterialFunctionCall_3.Result
   localLightDirection=@MaterialExpressionSkyAtmosphereLightDirection_1
   noise=@MaterialExpressionTextureObjectParameter_0
   timeAnim=@MaterialExpressionMultiply_6
   noiseStrength=@MaterialExpressionScalarParameter_9
   SphereRadius=@MaterialExpressionScalarParameter_7
   SpherePosition=@MaterialExpressionVectorParameter_2.RGB
   NormalizedSphereFalloff=@MaterialExpressionScalarParameter_8
   Streaming2DTex=@MaterialExpressionTextureObjectParameter_3
   AO_power=@MaterialExpressionScalarParameter_10
   innerSoftnessDistance=@MaterialExpressionScalarParameter_11
   DensityDarkenMin=@MaterialExpressionScalarParameter_12
   DensityDarkenPower=@MaterialExpressionScalarParameter_13
  ) {
   code="\r\n\r\nfloat theta = dot(normalize(localLightDirection), normalize(cameraVector));\r\nfloat3 cameraVectorOld = cameraVector;\r\ncameraVector *= stepSize;\r\n\r\n\r\nstruct Func {\r\n\r\n// the Henyey-Greenstein phase function\r\nfloat phase( float g,  float cos_theta)\r\n{\r\n    float denom = 1.0 + g * g - 2.0 * g * cos_theta;\r\n    return 1.0 / (4 * 3.14159) * (1.0 - g * g) / (denom * sqrt(denom));\r\n}\r\n\r\nfloat remapValueRange(float value, float low1, float high1, float low2,  float high2)\r\n{\r\n    return low2 + (value - low1) * (high2 - low2) / (high1 - low1);\r\n}\r\n\r\nfloat sphereDistance(float3 origin, float3 position, float radius){\r\n\r\n    //vector from position to sphere centre\r\n    float3 relativePosition = origin-position;\r\n\r\n    return length(relativePosition)-radius;\r\n}\r\n\r\n};\r\n\r\n\r\nfloat accumdist = 0;\r\nfloat current_density = 0;\r\nfloat transmittance = 1.0;\r\nfloat3 lightenergy = 0;\r\n\r\nFunc func;\r\n\r\nfloat height = GetPrimitiveData(Parameters.PrimitiveId).LocalObjectBoundsMax.z;\r\nfloat3 tim = (float3(timeAnim, timeAnim, timeAnim));\r\n\r\nfor (int i = 0; i < numSteps; i++){\r\n    //float4 noiseSample = noise.SampleLevel( noiseSampler, (rayStart + float3(0,-timeAnim*0.6, -timeAnim)/2),0);\r\n    float4 noiseSample = PseudoVolumeTexture( noise, noiseSampler, (rayStart + float3(0,-timeAnim*0.6, -timeAnim)/2), float2(16, 8), 128, 0);\r\n    //float4 cursample = VolumeTexture.SampleLevel( VolumeTextureSampler, saturate(rayStart-(noiseSample.bga-0.5)*noiseStrength),0);\r\n    // float4 PseudoVolumeTexture(Texture2D Tex, SamplerState TexSampler, float3 inPos, float2 xysize, float numframes,\r\n\t//uint mipmode = 0, float miplevel = 0, float2 InDDX = 0, float2 InDDY = 0)\r\n    //float4 cursample = PseudoVolumeTexture( Streaming2DTex, Streaming2DTexSampler, saturate(rayStart-(noiseSample.bga-0.5)*noiseStrength), float2(128, 128), 121 );\r\n    float3 pos = saturate(rayStart-(noiseSample.bga-0.5)*noiseStrength);\r\n    float4 cursample = PseudoVolumeTexture( Streaming2DTex, Streaming2DTexSampler, pos, float2(11, 11), 121 );\r\n    float dens = func.remapValueRange(cursample.r, innerSoftnessDistance, 0.56, 1, 0);\r\n    float darken =  func.remapValueRange(saturate(pow(1.0-dens,DensityDarkenPower)), 0, 1,DensityDarkenMin, 1);\r\n    //dens *= saturate(func.remapValueRange(func.sphereDistance(SpherePosition, pos, SphereRadius), -SphereRadius*NormalizedSphereFalloff, 0.0, 0, 1));\r\n    //float2 curShadowsample = ShadowTexture.SampleLevel( VolumeTextureSampler, saturate(rayStart-(noiseSample.bga-0.5)*noiseStrength),0).rg;\r\n    float2 curShadowsample = PseudoVolumeTexture( ShadowTexture, ShadowTextureSampler, saturate(rayStart-(noiseSample.bga-0.5)*noiseStrength),float2(8, 8), 64, 0).rg;\r\n    if(dens > 0.001){\r\n\r\n        current_density = saturate(dens)  * densityGain    *  stepSize;\r\n        float t_i = 1- exp(-current_density  );\r\n        \r\n        lightenergy += lightColor * t_i * transmittance * pow(curShadowsample.r, shadowGain) * darken;// * func.phase(0.1,theta)*1/stepSize ;\r\n        lightenergy += ambientColor * t_i * transmittance * pow(saturate(curShadowsample.g), AO_power) ;\r\n        //lightenergy += 5*pow(saturate(func.remapValueRange(func.sphereDistance(SpherePosition+float3(0,0.1,0), rayStart, SphereRadius), -SphereRadius*NormalizedSphereFalloff, 0.0, 0.1, 0)),2);\r\n        transmittance *= 1- t_i;  \r\n        \r\n        \r\n    }\r\n    rayStart += cameraVector;\r\n}\r\n\r\n\r\n// final step to remove artefact\r\n{\r\n    rayStart -= cameraVector;\r\n    rayStart += cameraVectorOld*finalDelta;\r\n\r\n    //float4 noiseSample = noise.SampleLevel( noiseSampler, (rayStart + float3(0,-timeAnim*0.6, -timeAnim)/2),0);\r\n    float4 noiseSample = PseudoVolumeTexture( noise, noiseSampler, (rayStart + float3(0,-timeAnim*0.6, -timeAnim)/2), float2(16, 8), 128, 0);\r\n    //float4 cursample = VolumeTexture.SampleLevel( VolumeTextureSampler, saturate(rayStart-(noiseSample.bga-0.5)*noiseStrength),0);\r\n    float3 pos = saturate(rayStart-(noiseSample.bga-0.5)*noiseStrength);\r\n    float4 cursample = PseudoVolumeTexture( Streaming2DTex, Streaming2DTexSampler,  pos, float2(11, 11), 121 );\r\n    float dens = func.remapValueRange(cursample.a, innerSoftnessDistance, 0.56, 1, 0);\r\n    float darken =  func.remapValueRange(saturate(pow(1.0-dens,DensityDarkenPower)), 0, 1,DensityDarkenMin, 1);\r\n    //dens *= saturate(func.remapValueRange(func.sphereDistance(SpherePosition, pos, SphereRadius), -SphereRadius*NormalizedSphereFalloff, 0.0, 0, 1));\r\n    //float2 curShadowsample = ShadowTexture.SampleLevel( VolumeTextureSampler,saturate(rayStart-(noiseSample.bga-0.5)*noiseStrength),0).rg;\r\n    float2 curShadowsample = PseudoVolumeTexture( ShadowTexture, ShadowTextureSampler, saturate(rayStart-(noiseSample.bga-0.5)*noiseStrength),float2(8, 8), 64, 0).rg;\r\n    if(dens > 0.001){\r\n\r\n        current_density = saturate(dens)  * densityGain    *  stepSize;\r\n        float t_i = 1- exp(-current_density  );\r\n        \r\n        lightenergy += lightColor * t_i * transmittance * pow(curShadowsample.r, shadowGain) * darken;// * func.phase(0.1,theta)*1/stepSize ;\r\n        lightenergy += ambientColor * t_i * transmittance *  pow(saturate(curShadowsample.g), AO_power);\r\n        transmittance *= 1- t_i;  \r\n        \r\n        \r\n    }\r\n}\r\n\r\nreturn float4( lightenergy, 1-transmittance);\r\n\r\n"
   description=DensityRayMarch
   outputType=CMOT_Float4
   showCode=true
   inputs [
    {
     inputName=rayStart
     input {expression=&self:MaterialExpressionNamedRerouteUsage_21}
    }
    {
     inputName=numSteps
     input {expression=&self:MaterialExpressionNamedRerouteUsage_20}
    }
    {
     inputName=cameraVector
     input {expression=&self:MaterialExpressionNamedRerouteUsage_3}
    }
    {
     inputName=stepSize
     input {expression=&self:MaterialExpressionNamedRerouteUsage_23}
    }
    {
     inputName=finalDelta
     input {expression=&self:MaterialExpressionNamedRerouteUsage_24}
    }
    {
     inputName=ShadowTexture
     input {expression=&self:MaterialExpressionTextureObjectParameter_1}
    }
    {
     inputName=densityGain
     input {expression=&self:MaterialExpressionScalarParameter_1}
    }
    {
     inputName=shadowGain
     input {expression=&self:MaterialExpressionScalarParameter_3}
    }
    {
     inputName=ambientColor
     input {expression=&self:MaterialExpressionMaterialFunctionCall_4}
    }
    {
     inputName=lightColor
     input {expression=&self:MaterialExpressionMaterialFunctionCall_3}
    }
    {
     inputName=localLightDirection
     input {expression=&self:MaterialExpressionSkyAtmosphereLightDirection_1}
    }
    {
     inputName=noise
     input {expression=&self:MaterialExpressionTextureObjectParameter_0}
    }
    {
     inputName=timeAnim
     input {expression=&self:MaterialExpressionMultiply_6}
    }
    {
     inputName=noiseStrength
     input {expression=&self:MaterialExpressionScalarParameter_9}
    }
    {
     inputName=SphereRadius
     input {expression=&self:MaterialExpressionScalarParameter_7}
    }
    {
     inputName=SpherePosition
     input {
      expression=&self:MaterialExpressionVectorParameter_2
      mask=1
      maskB=1
      maskG=1
      maskR=1
     }
    }
    {
     inputName=NormalizedSphereFalloff
     input {expression=&self:MaterialExpressionScalarParameter_8}
    }
    {
     inputName=Streaming2DTex
     input {expression=&self:MaterialExpressionTextureObjectParameter_3}
    }
    {
     inputName=AO_power
     input {expression=&self:MaterialExpressionScalarParameter_10}
    }
    {
     inputName=innerSoftnessDistance
     input {expression=&self:MaterialExpressionScalarParameter_11}
    }
    {
     inputName=DensityDarkenMin
     input {expression=&self:MaterialExpressionScalarParameter_12}
    }
    {
     inputName=DensityDarkenPower
     input {expression=&self:MaterialExpressionScalarParameter_13}
    }
   ]
  }
  ComponentMask(None=@n6) {a=true}
  Constant()
  Constant()
  Constant() {r=1}
  Constant() {r=1}
  Custom!(None) {
   code=" float scale = length( TransformLocalVectorToWorld(Parameters, float3(1.00000000,0.00000000,0.00000000)).xyz);\r\nfloat localscenedepth = CalcSceneDepth(ScreenAlignedPosition(GetScreenPosition(Parameters)));\r\n\r\nfloat3 camerafwd = mul(float3(0.00000000,0.00000000,1.00000000),ResolvedView.ViewToTranslatedWorld);\r\nlocalscenedepth /= (GetPrimitiveData(Parameters.PrimitiveId).LocalObjectBoundsMax.x * 2 * scale);\r\nlocalscenedepth /= abs( dot( camerafwd, Parameters.CameraVector ) );\r\nreturn localscenedepth;\r\n\r\n"
   description=camfwd
   outputType=CMOT_Float1
  }
  Custom() {
   code="int3 randpos = int3(Parameters.SvPosition.xy, View.StateFrameIndexMod8);\r\nfloat rand =float(Rand3DPCG16(randpos).x) / 0xffff;\r\nreturn rand.x;"
   showCode=true
  }
  EyeAdaptationInverse()
  EyeAdaptation()
  FeatureLevelSwitch()
  n7 = Custom!(in=@MaterialExpressionTime_0) {
   code="// The below expression will get compiled\r\n// into the output of this node\r\nreturn in;"
   outputType=CMOT_Float1
   showCode=true
   inputs [{inputName=in, input {expression=&self:MaterialExpressionTime_0}}]
  }
  Frac(None=@n7)
  n8 = ComponentMask(None=@n6) {b=true, g=true, r=true}
  n9 = ScalarParameter() {defaultValue=1, parameterName=MinBrightness}
  n10 = ScalarParameter() {defaultValue=1.35, parameterName=MaxBrightness}
  n11 = DynamicParameter() {defaultValue=<rgba 0 1 1 1>, paramNames [FlashIntensity, Param2, Param3, Param4]}
  n12 = MaterialFunctionCall(
   In=DotProduct(A=Constant3Vector() {constant=<rgba 0.21 0.71 0.07 0>}, B=@n8)
   Contrast=ScalarParameter() {defaultValue=0.22, parameterName=FlashMaskContrast}
  ) {materialFunction=&CheapContrast}
  n13 = ScalarParameter() {defaultValue=6, parameterName=FlashMaskMultiplier}
  LinearInterpolate(
   A=@n8
   B=Multiply(A=@n8, B=LinearInterpolate(A=@n9, B=@n10, Alpha=@n11.FlashIntensity))
   Alpha=Saturate(None=Multiply(A=@n12.Result, B=@n13))
  ) {constAlpha=1}
  MaterialFunctionCall!(Sine Phase, Value 1, Value 2) {materialFunction=/Engine/Functions/Engine_MaterialFunctions02/Utility/Sine_Remapped}
  MaterialFunctionCall() {materialFunction=&ObjectLocalBounds}
  MaterialFunctionCall() {materialFunction=&ObjectLocalBounds}
  n14 = Transform(None=CameraVectorWS()) {transformSourceType=TRANSFORMSOURCE_World, transformType=TRANSFORM_Instance}
  NamedRerouteDeclaration(None=Multiply(A=Normalize(VectorInput=@n14)) {constB=-1}) {name=Local Camera Vector, nodeColor=<rgba 0 1 0.5781 1>}
  n15 = CameraPositionWS()
  n16 = Constant()
  n17 = ScalarParameter() {parameterName=VerticalOffset}
  n18 = Divide(
   A=Add(
    A=TransformPosition(None=@n15) {transformSourceType=TRANSFORMPOSSOURCE_World, transformType=TRANSFORMPOSSOURCE_Instance}
    B=MaterialFunctionCall(X=@n16, Y=@n16, Z=@n17) {materialFunction=&MakeFloat3}.Result
   )
   B=ComponentMask(None=PreSkinnedLocalBounds().Max) {r=true}
  )
  n19 = Multiply(A=@n18) {constB=0.5}
  NamedRerouteDeclaration(None=Add(A=@n19) {constB=0.5}) {name=Local Camera Position, nodeColor=<rgba 0.2891 0 1 1>}
  n20 = NamedRerouteUsage()
  n21 = Add(A=@n20, B=NamedRerouteUsage())
  n22 = Max(A=MaterialFunctionCall("Xb1,PS4"=@n20, "PC-High&&Epic,XSX,PS5"=@n20, "PC-Perf,Android"=@n21, Switch=@n20, "PC-Low(NotPerf)&&Medium"=@n21) {materialFunction=&PlatformSwitch_1D}.Result) {constB=0}
  n23 = Add(A=NamedRerouteUsage(), B=Multiply(A=NamedRerouteUsage(), B=@n22))
  n24 = Multiply(A=NamedRerouteUsage(), B=NamedRerouteUsage())
  n25 = Add(A=@n23, B=Multiply(A=NamedRerouteUsage(), B=@n24))
  NamedRerouteDeclaration(None=MaterialFunctionCall("Xb1,PS4"=@n25, "PC-High&&Epic,XSX,PS5"=@n25, "PC-Perf,Android"=@n23, Switch=@n25, "PC-Low(NotPerf)&&Medium"=@n23) {materialFunction=&PlatformSwitch_3D}.Result) {name=Ray Start Position, nodeColor=<rgba 1 0 0.9141 1>}
  n26 = Floor(None=@n5)
  n27 = Multiply(A=@n4, B=@n26)
  NamedRerouteDeclaration(None=Subtract(A=NamedRerouteUsage(), B=Add(A=@n22, B=@n27))) {name=Final Delta Distance, nodeColor=<rgba 1 0.9375 0 1>}
  NamedRerouteDeclaration(None=@n18) {name=Normalized Camera Position, nodeColor=<rgba 0 1 0.7891 1>}
  NamedRerouteDeclaration(None=@n24) {name=BlueStep, nodeColor=<rgba 0.027 0 1 1>}
  NamedRerouteDeclaration(None=MaterialFunctionCall(Vector 3=@n19) {materialFunction=/Engine/Functions/Engine_MaterialFunctions02/Utility/VectorLength}.V3 Length) {name=Camera Distance From Centre of Vol Box, nodeColor=<rgba 0.5 1 0 1>}
  NamedRerouteDeclaration(None=MaterialFunctionCall!(Alpha Threshold, Random) {materialFunction=/Game/TimeOfDay/CloudsAndStorm/Materials/MaterialFunctions/DitherTemporalAASpatioTemporalBlue}.Result) {name=BlueNoise, nodeColor=<rgba 1 0 0.7734 1>}
  n28 = Divide(A=Constant() {r=1.7321}, B=NamedRerouteUsage())
  NamedRerouteDeclaration(None=@n28) {name=Step Size, nodeColor=<rgba 1 0.9844 0 1>}
  n29 = Divide(A=Constant() {r=1}, B=NamedRerouteUsage())
  n30 = Multiply(A=Subtract(A=Constant(), B=NamedRerouteUsage()), B=@n29)
  n31 = Multiply(A=Subtract(A=Constant() {r=1}, B=NamedRerouteUsage()), B=@n29)
  n32 = MaterialFunctionCall(Float3=Min(A=@n30, B=@n31)) {materialFunction=&BreakOutFloat3Components}
  n33 = Max(A=@n32.G, B=@n32.G)
  NamedRerouteDeclaration(None=Max(A=Max(A=@n32.R, B=@n33)) {constB=0}) {name=Near distance t_0, nodeColor=<rgba 0.875 0 1 1>}
  n34 = ScalarParameter() {defaultValue=24, parameterName=maxSteps}
  n35 = Multiply(A=@n34, B=ScalarParameter() {defaultValue=0.5, parameterName=Mobile Quality Multipler})
  NamedRerouteDeclaration(None=MaterialFunctionCall("Xb1,PS4"=@n34, "PC-High&&Epic,XSX,PS5"=@n34, "PC-Perf,Android"=@n35, Switch=@n34, "PC-Low(NotPerf)&&Medium"=@n35) {materialFunction=&PlatformSwitch_1D}.Result) {name=Steps, nodeColor=<rgba 0.8516 0 1 1>}
  n36 = MaterialFunctionCall(Float3=Max(A=@n30, B=@n31)) {materialFunction=&BreakOutFloat3Components}
  n37 = Min(A=@n36.R, B=Min(A=@n36.G, B=@n36.G))
  n38 = MaterialFunctionCall() {materialFunction=&ScreenAlignedUVs}
  n39 = Multiply(A=ComponentMask(None=MaterialFunctionCall() {materialFunction=/Game/Animation/Props/Gadgets_LiftingAura/Materials/MaterialFunctions/MF_InstanceBounds}.Instance Bounds Max) {r=true}) {constB=2}
  n40 = MaterialFunctionCall(Vector 3=Transform(None=Constant3Vector() {constant=<rgba 1 0 0 0>}) {transformSourceType=TRANSFORMSOURCE_Instance}) {materialFunction=&VectorLength}
  n41 = CameraVectorWS()
  n42 = TransformPosition(None=@n1) {transformSourceType=TRANSFORMPOSSOURCE_View, transformType=TRANSFORMPOSSOURCE_TranslatedWorld}
  n43 = Divide(A=Divide(A=SceneDepth(UVs=@n38."X 100%, Y 100%").R, B=Multiply(A=@n39, B=@n40.V3 Length) {constB=2}), B=Abs(None=DotProduct(A=@n41, B=@n42)))
  NamedRerouteDeclaration(None=StaticSwitchParameter(True=Min(A=@n37, B=@n43), False=@n37) {defaultValue=true, parameterName=Composite Against Scene Depth}) {name=Far Distance t_1, nodeColor=<rgba 1 0 0.3984 1>}
  n44 = Subtract(A=NamedRerouteUsage(), B=NamedRerouteUsage())
  n45 = Multiply(A=NamedRerouteUsage()) {constB=0.5}
  n46 = Subtract(A=Constant() {r=1}, B=Frac(None=Multiply(A=@n44, B=@n45)))
  NamedRerouteDeclaration(None=Multiply(A=Multiply(A=@n46, B=@n28) {constB=0.2}, B=ScalarParameter() {defaultValue=1, parameterName=planeAlignment, sliderMax=1})) {name=Plane offset, nodeColor=<rgba 1 0 0.7734 1>}
  NamedRerouteDeclaration(None=@n26) {name=Step Count, nodeColor=<rgba 0 0.3906 1 1>}
  NamedRerouteUsage()
  NamedRerouteUsage()
  NamedRerouteUsage()
  n47 = Constant() {r=1}
  n48 = Add(A=NamedRerouteUsage(), B=Multiply(A=NamedRerouteUsage(), B=NamedRerouteUsage()))
  n49 = DotProduct(A=Subtract(A=Multiply(A=@n48) {constB=2}), B=Normalize(VectorInput=NamedRerouteUsage()))
  n50 = Subtract(A=OneMinus(None=Divide(A=@n49) {constB=1.73})) {constB=0.866}
  n51 = Frac(None=Multiply(A=Multiply(A=@n50) {constB=0.5}, B=ScalarParameter() {defaultValue=24, parameterName=maxSteps}))
  StaticSwitchParameter(False=Multiply(A=Subtract(A=@n47, B=@n51), B=@n28) {constB=0.2}) {defaultValue=true, parameterName=SphericalSlice}
  Transform(None=SkyAtmosphereLightDirection()) {transformSourceType=TRANSFORMSOURCE_World, transformType=TRANSFORM_Local}
  Transform(None=CameraPositionWS()) {transformSourceType=TRANSFORMSOURCE_World, transformType=TRANSFORM_Local}
  ViewProperty() {property=MEVP_TemporalSampleIndex}
 ]
}
